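# Unlit translucent material that draws a soft colored glow sprite, apparently for the TeaCake shard environment effect.
# World position offset combines a sprite function, scaled optionally by camera distance, with a camera offset.
# Emissive color is modulated by time of day.
# Opacity comes from a smoothstepped UV sphere mask with optional distance fade, plus depth fade that the Low quality level skips.
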Material /Game/Effects/Fort_Effects/Effects/Environment/Hera/Shards/Glow/M_Env_TeaCake_Shard_SoftGlow
graph {
 properties {bAutomaticallySetUsageInEditor=false, blendMode=BLEND_Translucent, shadingModel=MSM_Unlit, usedShadingModels=Unlit}
 nodes [
  n1 = ScalarParameter() {defaultValue=20000, group=Scale, parameterName=SpriteScaleMult}
  n2 = LinearInterpolate(
   B=MaterialFunctionCall(
    Fade_Start_Distance=ScalarParameter() {defaultValue=30000, group=Scale, parameterName=SpriteScale_DistanceFade_EndFading}
    Fade_End_Distance=ScalarParameter() {defaultValue=10000, group=Scale, parameterName=SpriteScale_DistanceFade_StartFading}
   ) {materialFunction=&MF_Distance_SyncedFade_FX}.Result
   Alpha=ScalarParameter() {defaultValue=1, group=Scale, parameterName=UseSpriteScaleDistanceFade?}
  ) {constA=1}
  Add(
   A=MaterialFunctionCall!(Center Location, 0-1 UVs, XY scale=Multiply(A=@n1, B=@n2), OverridePivot, "Normals (Optional)") {materialFunction=&Sprite}.World position offset
   B=MaterialFunctionCall(Offset Amount=ScalarParameter() {defaultValue=2000, parameterName=CameraOffset}) {materialFunction=/Engine/Functions/Engine_MaterialFunctions02/WorldPositionOffset/CameraOffset}.Camera Offset
  )
  n3 = ScalarParameter() {defaultValue=1, group=TimeOfDay, parameterName=MultTOD_Morning}
  n4 = ScalarParameter() {defaultValue=1, group=TimeOfDay, parameterName=MultTOD_Day}
  n5 = ScalarParameter() {defaultValue=1, group=TimeOfDay, parameterName=MultTOD_Evening}
  n6 = ScalarParameter() {defaultValue=0.6, group=TimeOfDay, parameterName=MultTOD_Night}
  MaterialFunctionCall(Emissive=Multiply(
    A=MaterialFunctionCall(Value_Morning=@n3, Value_Day=@n4, Value_Evening=@n5, Value_Night=@n6) {materialFunction=&Scalar_Time_of_Day}.Result
    B=VectorParameter() {parameterName=Color, defaultValue=<rgba 3.8118 0.1764 5 1>}.RGB
   )) {materialFunction=&MF_FX_EmissiveAdjustments}
  n7 = TextureCoordinate()
  n8 = ScalarParameter() {defaultValue=0.5, parameterName=SphereMask_Falloff}
  n9 = ScalarParameter() {defaultValue=0.5, parameterName=SphereMask_Radius}
  n10 = ScalarParameter() {parameterName=SphereMask_Hardness}
  n11 = Multiply(
   A=SmoothStep(
    Min=ScalarParameter() {parameterName=SmoothStep_SphereMask_Min}
    Max=ScalarParameter() {defaultValue=3, parameterName=SmoothStep_SphereMask_Max}
    Value=SphereMask(A=@n7, B=@n8, Radius=@n9, Hardness=@n10)
   )
   B=ScalarParameter() {defaultValue=0.5, parameterName=MultOpacity}
  )
  n12 = ScalarParameter() {defaultValue=80000, group=Scale, parameterName=DistanceFade_EndFading}
  n13 = ScalarParameter() {defaultValue=18000, group=Scale, parameterName=DistanceFade_StartFading}
  n14 = LinearInterpolate(
   A=@n11
   B=LinearInterpolate(
    B=@n11
    Alpha=MaterialFunctionCall(Fade_Start_Distance=@n12, Fade_End_Distance=@n13) {materialFunction=&MF_Distance_SyncedFade_FX}.Result
   ) {constB=0}
   Alpha=ScalarParameter() {defaultValue=1, group=Scale, parameterName=UseDistanceFade?}
  )
  n15 = Multiply(A=@n14, B=DepthFade(FadeDistance=ScalarParameter() {defaultValue=2000, parameterName=DepthFade}))
  QualitySwitch(Default=@n15, Low=@n14, High=@n15, Medium=@n15, Epic=@n15)
 ]
}
